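; Repair mod box that, when it dies, repairs and refuels the vehicle that dealt the damage, unless that vehicle is at critical health.
; The death handler fixes parts, applies a SetByCaller healing effect, adds fuel and plays effects.
; An attached Cowcatcher, found by gameplay tag, is restored to full health.
; Remote instances make the tagless box's world UI skip its mod check so the label shows in vehicles.
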(event EventBeginPlay
  (|CallParentFunction)
  (WorldUI|SetLabelText "NSLOCTEXT(\"[4A21CFDA953F5B8AFF350BB9474A3764]\", \"8D0A06B7488A6A5D70A1BC83740901B4\", \"REPAIR\")" "" (Variables|Default|GetModIcon))
  (EventDispatchers|AssignOnDied (delegate Custom|CustomEvent))
  (Networking|SwitchHasAuthority
    (:Remote
      (bind _aswbp_mod_box_world_ui (Utilities|Casting|CastToWBP_ModBox_WorldUI (UserInterface|GetWidget (Variables|Default|GetFortWidget))))
      (Variables|/VehicleMods/Gameplay/Actors/ModBox/UI/WBP_ModBox_WorldUI.WBP_ModBox_WorldUI_C|SetbSkipCanApplyModCheck true _aswbp_mod_box_world_ui) ; This box has no mod tag, so without this the box name UI won't show while in a vehicle.
    )))

(event Custom|CustomEvent
  (params (DamagedActor :type Actor%20Object%20Reference) (Damage :type Float%20%28single-precision%29) (InstigatedBy :type Controller%20Object%20Reference) (DamageCauser :type Actor%20Object%20Reference) (HitLocation :type Vector) (FHitComponent :type Primitive%20Component%20Object%20Reference) (BoneName :type Name) (Momentum :type Vector))
  (bind _returnvalue (Ability|GetAbilitySystemComponent (Variables|Default|GetVehicle)))
  (bind _vehicle (Variables|Default|GetVehicle))
  ;[1 1 1] Repair Vehicle and Parts, and refuel
  (bind _asfort_athena_vehicle (Utilities|Casting|CastToFortAthenaVehicle DamageCauser))
  (bind _output_get (Variables|Default|SetVehicle _asfort_athena_vehicle))
  (if (not (Vehicle|Destruction|IsCriticalHealth _output_get))
    (Class|FortAthenaVehicle|RepairAllDamageableParts _vehicle 1.0) ; Repair Glass and Popped Tires
    (bind _returnvalue_2 (Ability|GameplayEffect|AssignTagSetbyCallerMagnitude (GameplayEffects|MakeOutgoingSpec _returnvalue (Variables|Default|GetGE_HealEffectToApply) 1.0 (Class|AthenaFunctionLibrary|MakeSimpleFortEffectContext _vehicle)) "(TagName=\"SetByCaller.Healing\")" (* (Class|FortVehicleInterface|GetMaxHealth _vehicle) (ScalableFloat|GetValueatLevel (Variables|Default|GetRowHealPercent)))))
    (GameplayEffects|ApplyGameplayEffectSpecToSelf _returnvalue _returnvalue_2) ; Heal
    (Vehicle|Fuel|SetFuel _vehicle (+ (* (Vehicle|Fuel|GetFuelCapacity _vehicle) (ScalableFloat|GetValueatLevel (Variables|Default|GetRowFuelPercent))) (Vehicle|Fuel|GetFuel _vehicle))) ; Refuel
    (Niagara|SpawnSystemAttached "/Valet/FX/Niagara/NS_VehicleRepair_Instant.NS_VehicleRepair_Instant" (Class|FortVehicleInterface|GetSkeletalMeshComponent _vehicle) "None" "0, 0, 0" "0, 0, 0" "KeepRelativeOffset" true)
    (GameplayCue|ExecuteGameplayCue _vehicle (Variables|Default|GetGC_HealVehicle))
    ;;[0 0 0] Heal the Cowcatcher if the vehicle has one
    (if (Vehicle|Mod|HasVehicleModApplied _vehicle (Variables|Default|GetCowCatcherModTag))
      (for _array_element (Actor|GetAttachedActors _vehicle true false)
        (bind _asbuilding_gameplay_actor (Utilities|Casting|CastToBuildingGameplayActor _array_element)) ; Are we a BGA?
        (if (GameplayTags|HasMatchingGameplayTag _asbuilding_gameplay_actor "(TagName=\"Vehicle.Mod.Cowcatcher\")")
          (Health|SetHealth _asbuilding_gameplay_actor (Health|GetMaxHealth _asbuilding_gameplay_actor)) ; Heal CC to 100%
        )))))

(fn ConstructionScript)
